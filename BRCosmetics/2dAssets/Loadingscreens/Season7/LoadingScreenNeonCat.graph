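# Season 7 Neon Cat loading screen widget: a neon zigzag, drip and spray backdrop with rotating character art on the right.
# On Construct, Intro fades out a black overlay and animates zigzag and drip material reveals.
# The Timer animation serves as a repeating two-second clock that triggers SwapImage, which wipes between two stacked Images through material parameters.
# When each swap finishes, the bottom Image takes the new material and ActiveIndex advances through ImageArray, wrapping to zero.
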WidgetBlueprint /BRCosmetics/2dAssets/Loadingscreens/Season7/LoadingScreenNeonCat
animations [
 SwapImage {
  binding_count=2
  length_seconds=0.9
  object=&self:SwapImage
  sequence {
   bindings [
    Image_CharacterTop {
     id=#1
     possessed_class=Image
     tracks [{
       type=MovieSceneWidgetMaterialTrack
       sections [{type=MovieSceneParameterSection, channels [Reveal {type=MovieSceneScriptingFloatChannel, keys [[0,-0.1],[54,1.1]]}]}]
      }]
    }
    Image_CharacterBottom {
     id=#2
     possessed_class=Image
     tracks [{
       type=MovieSceneWidgetMaterialTrack
       sections [{type=MovieSceneParameterSection, channels [UnReveal {type=MovieSceneScriptingFloatChannel, keys [[0,1],[54,0]]}]}]
      }]
    }
   ]
   display_rate {numerator=60}
  }
 }
 Timer {
  binding_count=1
  length_seconds=1.997
  object=&self:Timer
  sequence {
   bindings [Overlay_0 {
     id=#3
     possessed_class=Overlay
     tracks [{
       type=MovieSceneFloatVectorTrack
       display_name=Pivot
       sections [{
         type=MovieSceneFloatVectorSection
         end_frame=39
         start_frame=0
         channels [
          Vector.X {type=MovieSceneScriptingFloatChannel, const [0.5,2]}
          Vector.Y {type=MovieSceneScriptingFloatChannel, const [0.5,2]}
         ]
        }]
      }]
    }]
   tracks [{
     type=MovieSceneEventTrack
     sections [{
       type=MovieSceneEventTriggerSection
       channels [None {type=MovieSceneScriptingEventChannel, const ["<Struct 'MovieSceneEvent' (0x46662ea2760) {}>", 1]}]
      }]
    }]
  }
 }
 Intro {
  binding_count=7
  length_seconds=2.8057
  object=&self:Intro
  sequence {bindings [
    Image_ZigZag_Top {
     id=#4
     possessed_class=Image
     tracks [{
       type=MovieSceneWidgetMaterialTrack
       sections [{
         type=MovieSceneParameterSection
         channels [
          Angle1Reveal {type=MovieSceneScriptingFloatChannel, keys [[3.7173,1],[11.0787,0]]}
          Angle2Reveal {type=MovieSceneScriptingFloatChannel, keys [[11.0787,1],[18.44,0]]}
          Angle3Reveal {type=MovieSceneScriptingFloatChannel, keys [[18.4403,1],[25.8017,0]]}
         ]
        }]
      }]
    }
    Image_Drip3 {
     id=#5
     possessed_class=Image
     tracks [{
       type=MovieSceneWidgetMaterialTrack
       sections [{type=MovieSceneParameterSection, channels [Reveal {type=MovieSceneScriptingFloatChannel, keys [[3.783,1],[56.114,0]]}]}]
      }]
    }
    Image_ZigZag_Bottom {
     id=#6
     possessed_class=Image
     tracks [{
       type=MovieSceneWidgetMaterialTrack
       sections [{
         type=MovieSceneParameterSection
         channels [
          Angle1Reveal {type=MovieSceneScriptingFloatChannel, keys [[3.7173,1],[11.0787,0]]}
          Angle2Reveal {type=MovieSceneScriptingFloatChannel, keys [[11.0787,1],[18.44,0]]}
          Angle3Reveal {type=MovieSceneScriptingFloatChannel, keys [[18.4403,1],[25.8017,0]]}
         ]
        }]
      }]
    }
    Image_Drip1 {
     id=#7
     possessed_class=Image
     tracks [{
       type=MovieSceneWidgetMaterialTrack
       sections [{type=MovieSceneParameterSection, channels [Reveal {type=MovieSceneScriptingFloatChannel, keys [[3.783,1],[56.114,0]]}]}]
      }]
    }
    Image_ZigZag_Top_Top {
     id=#8
     possessed_class=Image
     tracks [{
       type=MovieSceneWidgetMaterialTrack
       sections [{type=MovieSceneParameterSection, channels [Angle3Reveal {type=MovieSceneScriptingFloatChannel, keys [[18.4403,1],[25.8017,0]]}]}]
      }]
    }
    Image_Fade {
     id=#9
     possessed_class=Image
     tracks [{
       type=MovieSceneColorTrack
       display_name=Color and Opacity
       sections [{
         type=MovieSceneColorSection
         channels [
          Color.R {type=MovieSceneScriptingFloatChannel, const [0,1]}
          Color.G {type=MovieSceneScriptingFloatChannel, const [0,1]}
          Color.B {type=MovieSceneScriptingFloatChannel, const [0,1]}
          Color.A {type=MovieSceneScriptingFloatChannel, keys [[0,1],[9.8423,0]]}
         ]
        }]
      }]
    }
    Image_Drip2 {
     id=#10
     possessed_class=Image
     tracks [{
       type=MovieSceneWidgetMaterialTrack
       sections [{type=MovieSceneParameterSection, channels [Reveal {type=MovieSceneScriptingFloatChannel, keys [[3.783,1],[56.114,0]]}]}]
      }]
    }
   ]}
 }
]
graphs [
 EventGraph {nodes [
   EventSwapImage() {is=title K2Node_CustomEvent}
   > PlayAnimation(InAnimation=Get SwapImage() -> (SwapImage:Widget Animation Object Reference))
   n1 = "Get (a copy)"(
    Array:Array of Material Instance Object References=Get ImageArray() -> (ImageArray:Array of Material Instance Object References)
    Dimension 1=Get ActiveIndex() -> (ActiveIndex:int=0)
   ) -> (Output:Material Instance Object Reference)
   > SetBrushFromMaterial(self=Get Image_CharacterTop() -> (Image_CharacterTop:Image Object Reference), Material=@n1)
   n2 = EventTimerComplete() {is=title K2Node_CustomEvent}
   > EventSwapImage() {is=title K2Node_CallFunction}
   > PlayAnimation(InAnimation=Get Timer() -> (Timer:Widget Animation Object Reference))
   n3 = EventSwapFinished() {is=title K2Node_CustomEvent}
   > SetBrushFromMaterial(self=Get Image_CharacterBottom() -> (Image_CharacterBottom:Image Object Reference), Material=@n1)
   n4 = int - int(A=Length(TargetArray:Array of Material Instance Object References=Get ImageArray()), B=1) {is=title K2Node_CallFunction}
   n5 = int + int(A=Get ActiveIndex() -> (ActiveIndex=0), B=1) -> (ReturnValue=0) {is=title K2Node_CommutativeAssociativeBinaryOperator}
   > Set ActiveIndex(ActiveIndex:int=Select(Option 0:int=@n5, Option 1:int=0, Index="integer > integer"(A=@n5, B=@n4) {is=title K2Node_CallFunction}) -> (ReturnValue:int)) -> (Output_Get:int=0)
   Event Construct()
   > BindToAnimationFinished(Animation=Get Timer(), Delegate=@n2)
   > BindToAnimationFinished(Animation=Get SwapImage(), Delegate=@n3)
   > PlayAnimation(InAnimation=Get Timer())
   > PlayAnimationForward(InAnimation=Get Intro())
   Event PreConstruct()
  ]}
 SequenceEvent_0 {nodes [SequenceEvent_0() {is=title K2Node_FunctionEntry}]}
]
variables [
 ActiveIndex {type=integer}
 ImageArray {
  type="array<MaterialInstance>"
  default [
   /BRCosmetics/2dAssets/Loadingscreens/Season7/NeonCatLoading/M_UI_Loading_NeonCatImageMaster_1
   /BRCosmetics/2dAssets/Loadingscreens/Season7/NeonCatLoading/M_UI_Loading_NeonCatImageMaster_2
   /BRCosmetics/2dAssets/Loadingscreens/Season7/NeonCatLoading/M_UI_Loading_NeonCatImageMaster_3
  ]
 }
]
widgets {
 info {inheritedWidgetCount=0, parentClass=&UserWidget, rootWidgetClass=&Overlay, widgetCount=19}
 widgets [
  Overlay_0 {widgetClassPath=&Overlay, widget=Overlay_0}
  ScaleBox_0 {
   widgetClassPath=&ScaleBox
   parent=Overlay_0
   widget=ScaleBox_0
   properties {stretch=ScaleToFitX, stretchDirection=DownOnly}
   slot {class=OverlaySlot, properties {horizontalAlignment=HAlign_Fill, verticalAlignment=VAlign_Fill}}
  }
  Overlay_2 {
   widgetClassPath=&Overlay
   parent=ScaleBox_0
   widget=Overlay_2
   slot {class=ScaleBoxSlot, properties {horizontalAlignment=HAlign_Fill, verticalAlignment=VAlign_Fill}}
  }
  Image_15 {
   widgetClassPath=&Image
   bIsVariable=true
   parent=Overlay_2
   widget=Image_15
   properties {brush {
     bIsDynamicallyLoaded=false
     drawAs=Image
     imageType=NoImage
     mirroring=NoMirror
     resourceName=None
     resourceObject=/BRCosmetics/2dAssets/Loadingscreens/Season7/NeonCatLoading/M_UI_NeonCatBG_Inst
     tiling=NoTile
     imageSize=<1920 1080>
     margin=<ltrb 0 0 0 0>
     outlineSettings {
      bUseBrushTransparency=false
      roundingType=HalfHeightRadius
      width=0
      color {colorUseRule=UseColor_Specified, specifiedColor=<rgba 0 0 0 0>}
      cornerRadii=<0 0 0 0>
     }
     tintColor {colorUseRule=UseColor_Specified, specifiedColor=<rgba 1 1 1 1>}
     uVRegion {bIsValid=false, max=<0 0>, min=<0 0>}
    }}
   slot {class=OverlaySlot, properties {horizontalAlignment=HAlign_Fill, verticalAlignment=VAlign_Fill}}
  }
  Image_LetterboxBottom {
   widgetClassPath=&Image
   parent=Overlay_2
   widget=Image_LetterboxBottom
   properties {
    brush {
     bIsDynamicallyLoaded=false
     drawAs=Image
     imageType=NoImage
     mirroring=NoMirror
     resourceName=None
     resourceObject=/BRCosmetics/2dAssets/Loadingscreens/Season7/NeonCatLoading/M_UI_HUD_LetterBoxSkew_Inst
     tiling=NoTile
     imageSize=<32 122.1887>
     margin=<ltrb 0 0 0 0>
     outlineSettings {
      bUseBrushTransparency=false
      roundingType=HalfHeightRadius
      width=0
      color {colorUseRule=UseColor_Specified, specifiedColor=<rgba 0 0 0 0>}
      cornerRadii=<0 0 0 0>
     }
     tintColor {colorUseRule=UseColor_Specified, specifiedColor=<rgba 1 1 1 1>}
     uVRegion {bIsValid=false, max=<0 0>, min=<0 0>}
    }
    colorAndOpacity=<rgba 0.011 0.0037 0.0091 1>
    renderTransform {angle=0, scale=<-1 -1>, shear=<0 0>, translation=<0 0>}
   }
   slot {class=OverlaySlot, properties {horizontalAlignment=HAlign_Fill, verticalAlignment=VAlign_Bottom}}
  }
  Overlay_4 {
   widgetClassPath=&Overlay
   parent=Overlay_2
   widget=Overlay_4
   slot {class=OverlaySlot, properties {horizontalAlignment=HAlign_Center, verticalAlignment=VAlign_Fill, padding=<ltrb -450 0 -400 -2000>}}
  }
  Image_ZigZag_Bottom {
   widgetClassPath=&Image
   parent=Overlay_4
   widget=Image_ZigZag_Bottom
   properties {
    brush {
     bIsDynamicallyLoaded=false
     drawAs=Image
     imageType=NoImage
     mirroring=NoMirror
     resourceName=None
     resourceObject=/BRCosmetics/2dAssets/Loadingscreens/Season7/NeonCatLoading/M_UI_ZigZag_Step_Inst2
     tiling=NoTile
     imageSize=<2142.269 1279.801>
     margin=<ltrb 0 0 0 0>
     outlineSettings {
      bUseBrushTransparency=false
      roundingType=HalfHeightRadius
      width=0
      color {colorUseRule=UseColor_Specified, specifiedColor=<rgba 0 0 0 0>}
      cornerRadii=<0 0 0 0>
     }
     tintColor {colorUseRule=UseColor_Specified, specifiedColor=<rgba 1 1 1 1>}
     uVRegion {bIsValid=false, max=<0 0>, min=<0 0>}
    }
    colorAndOpacity=<rgba 1 0.0685 0.0685 1>
    renderTransform {angle=0, scale=<1 1>, shear=<0 0>, translation=<-32 4>}
    renderTransformPivot=<0.5 0>
   }
   slot {class=OverlaySlot, properties {horizontalAlignment=HAlign_Center}}
  }
  Image_Drip2 {
   widgetClassPath=&Image
   parent=Overlay_4
   widget=Image_Drip2
   properties {
    brush {
     bIsDynamicallyLoaded=false
     drawAs=Image
     imageType=NoImage
     mirroring=NoMirror
     resourceName=None
     resourceObject=/BRCosmetics/2dAssets/Loadingscreens/Season7/NeonCatLoading/M_UI_LoadingNeonCat_Drips_Mat_Inst1
     tiling=NoTile
     imageSize=<45.2935 172.162>
     margin=<ltrb 0 0 0 0>
     outlineSettings {
      bUseBrushTransparency=false
      roundingType=HalfHeightRadius
      width=0
      color {colorUseRule=UseColor_Specified, specifiedColor=<rgba 0 0 0 0>}
      cornerRadii=<0 0 0 0>
     }
     tintColor {colorUseRule=UseColor_Specified, specifiedColor=<rgba 1 1 1 1>}
     uVRegion {bIsValid=false, max=<0 0>, min=<0 0>}
    }
    colorAndOpacity=<rgba 1 0.0685 0.0685 1>
    renderTransform {angle=0, scale=<1 1>, shear=<0 0>, translation=<880 307>}
   }
   slot {class=OverlaySlot, properties {horizontalAlignment=HAlign_Center}}
  }
  Image_ZigZag_Top {
   widgetClassPath=&Image
   parent=Overlay_2
   widget=Image_ZigZag_Top
   properties {
    brush {
     bIsDynamicallyLoaded=false
     drawAs=Image
     imageType=NoImage
     mirroring=NoMirror
     resourceName=None
     resourceObject=/BRCosmetics/2dAssets/Loadingscreens/Season7/NeonCatLoading/M_UI_ZigZag_Step_Inst
     tiling=NoTile
     imageSize=<1893.8137 32>
     margin=<ltrb 0 0 0 0>
     outlineSettings {
      bUseBrushTransparency=false
      roundingType=HalfHeightRadius
      width=0
      color {colorUseRule=UseColor_Specified, specifiedColor=<rgba 0 0 0 0>}
      cornerRadii=<0 0 0 0>
     }
     tintColor {colorUseRule=UseColor_Specified, specifiedColor=<rgba 1 1 1 1>}
     uVRegion {bIsValid=false, max=<0 0>, min=<0 0>}
    }
    colorAndOpacity=<rgba 0.0782 0.9131 1 1>
   }
   slot {class=OverlaySlot, properties {horizontalAlignment=HAlign_Center, verticalAlignment=VAlign_Fill, padding=<ltrb -450 0 -400 0>}}
  }
  Image_LetterboxTop {
   widgetClassPath=&Image
   parent=Overlay_2
   widget=Image_LetterboxTop
   properties {
    brush {
     bIsDynamicallyLoaded=false
     drawAs=Image
     imageType=NoImage
     mirroring=NoMirror
     resourceName=None
     resourceObject=&M_UI_HUD_LetterBoxSkew_Inst
     tiling=NoTile
     imageSize=<32 100.0596>
     margin=<ltrb 0 0 0 0>
     outlineSettings {
      bUseBrushTransparency=false
      roundingType=HalfHeightRadius
      width=0
      color {colorUseRule=UseColor_Specified, specifiedColor=<rgba 0 0 0 0>}
      cornerRadii=<0 0 0 0>
     }
     tintColor {colorUseRule=UseColor_Specified, specifiedColor=<rgba 1 1 1 1>}
     uVRegion {bIsValid=false, max=<0 0>, min=<0 0>}
    }
    colorAndOpacity=<rgba 0.011 0.0037 0.0091 1>
    renderTransformPivot=<0 0>
   }
   slot {class=OverlaySlot, properties {horizontalAlignment=HAlign_Fill}}
  }
  Image_Spray {
   widgetClassPath=&Image
   parent=Overlay_2
   widget=Image_Spray
   properties {brush {
     bIsDynamicallyLoaded=false
     drawAs=Image
     imageType=NoImage
     mirroring=NoMirror
     resourceName=None
     resourceObject=/BRCosmetics/2dAssets/Loadingscreens/Season7/NeonCatLoading/T_UI_Loading_NeonCat_Spray
     tiling=NoTile
     imageSize=<918 1055>
     margin=<ltrb 0 0 0 0>
     outlineSettings {
      bUseBrushTransparency=false
      roundingType=HalfHeightRadius
      width=0
      color {colorUseRule=UseColor_Specified, specifiedColor=<rgba 0 0 0 0>}
      cornerRadii=<0 0 0 0>
     }
     tintColor {colorUseRule=UseColor_Specified, specifiedColor=<rgba 1 1 1 1>}
     uVRegion {bIsValid=false, max=<0 0>, min=<0 0>}
    }}
   slot {class=OverlaySlot, properties {verticalAlignment=VAlign_Center, padding=<ltrb 8 0 0 0>}}
  }
  Overlay_3 {
   widgetClassPath=&Overlay
   parent=Overlay_2
   widget=Overlay_3
   slot {class=OverlaySlot, properties {horizontalAlignment=HAlign_Center, verticalAlignment=VAlign_Fill, padding=<ltrb -450 0 -400 0>}}
  }
  Image_ZigZag_Top_Top {
   widgetClassPath=&Image
   parent=Overlay_3
   widget=Image_ZigZag_Top_Top
   properties {
    brush {
     bIsDynamicallyLoaded=false
     drawAs=Image
     imageType=NoImage
     mirroring=NoMirror
     resourceName=None
     resourceObject=/BRCosmetics/2dAssets/Loadingscreens/Season7/NeonCatLoading/M_UI_ZigZag_Step_Inst_Inst_Top
     tiling=NoTile
     imageSize=<1893.8137 32>
     margin=<ltrb 0 0 0 0>
     outlineSettings {
      bUseBrushTransparency=false
      roundingType=HalfHeightRadius
      width=0
      color {colorUseRule=UseColor_Specified, specifiedColor=<rgba 0 0 0 0>}
      cornerRadii=<0 0 0 0>
     }
     tintColor {colorUseRule=UseColor_Specified, specifiedColor=<rgba 1 1 1 1>}
     uVRegion {bIsValid=false, max=<0 0>, min=<0 0>}
    }
    colorAndOpacity=<rgba 0.0782 0.9131 1 1>
   }
   slot {class=OverlaySlot, properties {horizontalAlignment=HAlign_Center, verticalAlignment=VAlign_Fill}}
  }
  Image_Drip1 {
   widgetClassPath=&Image
   parent=Overlay_3
   widget=Image_Drip1
   properties {
    brush {
     bIsDynamicallyLoaded=false
     drawAs=Image
     imageType=NoImage
     mirroring=NoMirror
     resourceName=None
     resourceObject=/BRCosmetics/2dAssets/Loadingscreens/Season7/NeonCatLoading/M_UI_LoadingNeonCat_Drips_Mat_Inst
     tiling=NoTile
     imageSize=<44.1315 192.2711>
     margin=<ltrb 0 0 0 0>
     outlineSettings {
      bUseBrushTransparency=false
      roundingType=HalfHeightRadius
      width=0
      color {colorUseRule=UseColor_Specified, specifiedColor=<rgba 0 0 0 0>}
      cornerRadii=<0 0 0 0>
     }
     tintColor {colorUseRule=UseColor_Specified, specifiedColor=<rgba 1 1 1 1>}
     uVRegion {bIsValid=false, max=<0 0>, min=<0 0>}
    }
    colorAndOpacity=<rgba 0.0782 0.9131 1 1>
    renderTransform {angle=0, scale=<1 1>, shear=<0 0>, translation=<834 228>}
   }
   slot {class=OverlaySlot, properties {horizontalAlignment=HAlign_Center}}
  }
  Image_Drip3 {
   widgetClassPath=&Image
   parent=Overlay_3
   widget=Image_Drip3
   properties {
    brush {
     bIsDynamicallyLoaded=false
     drawAs=Image
     imageType=NoImage
     mirroring=NoMirror
     resourceName=None
     resourceObject=/BRCosmetics/2dAssets/Loadingscreens/Season7/NeonCatLoading/M_UI_LoadingNeonCat_Drips_Mat_Inst2
     tiling=NoTile
     imageSize=<64.9152 203.7474>
     margin=<ltrb 0 0 0 0>
     outlineSettings {
      bUseBrushTransparency=false
      roundingType=HalfHeightRadius
      width=0
      color {colorUseRule=UseColor_Specified, specifiedColor=<rgba 0 0 0 0>}
      cornerRadii=<0 0 0 0>
     }
     tintColor {colorUseRule=UseColor_Specified, specifiedColor=<rgba 1 1 1 1>}
     uVRegion {bIsValid=false, max=<0 0>, min=<0 0>}
    }
    colorAndOpacity=<rgba 0.0782 0.9131 1 1>
    renderTransform {angle=0, scale=<1 1>, shear=<0 0>, translation=<-325 952>}
   }
   slot {class=OverlaySlot, properties {horizontalAlignment=HAlign_Center}}
  }
  Overlay_1 {
   widgetClassPath=&Overlay
   parent=Overlay_0
   widget=Overlay_1
   slot {class=OverlaySlot, properties {horizontalAlignment=HAlign_Right, verticalAlignment=VAlign_Bottom, padding=<ltrb 0 0 250 0>}}
  }
  Image_CharacterBottom {
   widgetClassPath=&Image
   bIsVariable=true
   parent=Overlay_1
   widget=Image_CharacterBottom
   properties {brush {
     bIsDynamicallyLoaded=false
     drawAs=Image
     imageType=NoImage
     mirroring=NoMirror
     resourceName=None
     resourceObject=/BRCosmetics/2dAssets/Loadingscreens/Season7/NeonCatLoading/M_UI_Loading_NeonCatImageMaster_Empty
     tiling=NoTile
     imageSize=<573 1080>
     margin=<ltrb 0 0 0 0>
     outlineSettings {
      bUseBrushTransparency=false
      roundingType=HalfHeightRadius
      width=0
      color {colorUseRule=UseColor_Specified, specifiedColor=<rgba 0 0 0 0>}
      cornerRadii=<0 0 0 0>
     }
     tintColor {colorUseRule=UseColor_Specified, specifiedColor=<rgba 1 1 1 1>}
     uVRegion {bIsValid=false, max=<0 0>, min=<0 0>}
    }}
   slot {class=OverlaySlot}
  }
  Image_CharacterTop {
   widgetClassPath=&Image
   bIsVariable=true
   parent=Overlay_1
   widget=Image_CharacterTop
   properties {brush {
     bIsDynamicallyLoaded=false
     drawAs=Image
     imageType=NoImage
     mirroring=NoMirror
     resourceName=None
     resourceObject=&M_UI_Loading_NeonCatImageMaster_Empty
     tiling=NoTile
     imageSize=<573 1080>
     margin=<ltrb 0 0 0 0>
     outlineSettings {
      bUseBrushTransparency=false
      roundingType=HalfHeightRadius
      width=0
      color {colorUseRule=UseColor_Specified, specifiedColor=<rgba 0 0 0 0>}
      cornerRadii=<0 0 0 0>
     }
     tintColor {colorUseRule=UseColor_Specified, specifiedColor=<rgba 1 1 1 1>}
     uVRegion {bIsValid=false, max=<0 0>, min=<0 0>}
    }}
   slot {class=OverlaySlot}
  }
  Image_Fade {
   widgetClassPath=&Image
   bIsVariable=true
   parent=Overlay_0
   widget=Image_Fade
   properties {colorAndOpacity=<rgba 0 0 0 0>}
   slot {class=OverlaySlot, properties {horizontalAlignment=HAlign_Fill, verticalAlignment=VAlign_Fill}}
  }
 ]
}
